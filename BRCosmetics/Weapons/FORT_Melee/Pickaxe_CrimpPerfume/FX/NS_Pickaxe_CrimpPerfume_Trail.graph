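# Niagara swing trail for the CrimpPerfume pickaxe, combining two anim-trail ribbons with falling leaf sprites.
# Head and Vines ribbons follow the same bone or socket pair, fading and scaling width over short lifetimes.
# Socket-spawned leaves inherit velocity, then tumble under curl noise, gravity and drag, with sprites aligned to mesh orientation and sized by camera distance.
# Time-of-day and front-end inputs adjust color and material response.
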NiagaraSystem /BRCosmetics/Weapons/FORT_Melee/Pickaxe_CrimpPerfume/FX/NS_Pickaxe_CrimpPerfume_Trail
dependencies {
 usedDataInterfaces [
  &NiagaraDataInterfaceArrayColor
  &NiagaraDataInterfaceCurve
  &NiagaraDataInterfaceCamera
  &NiagaraDataInterfaceSpriteRendererInfo
  &NiagaraDataInterfaceSocketReader
 ]
 usedDynamicInputs []
 usedModules [
  &self:SystemSpawnScript.SystemScriptSource.SystemScriptGraph.NiagaraNodeAssignment_0.SetVariables_#1
  &SystemState
  &self:SystemSpawnScript.SystemScriptSource.SystemScriptGraph.NiagaraNodeAssignment_1.SetVariables_#2
 ]
}
emitters {
 NE_Leaves {
  emitterUpdateScript [
   &EmitterState(Life Cycle Mode=::NewEnumerator1, Loop Behavior=::NewEnumerator1, Loop Duration=0.25, UseLoopDelay=false)
   &SpawnRate(SpawnRate=60, Use Spawn Probability=false)
  ]
  particleSpawnScript [
   &InitializeParticle(
    Color=&SelectLinearColorFromArray(
     Array Sampling Mode=::NewEnumerator0
     Direct Array Index=?
     Color Selection Array=DataInterface() {propertyValues {ColorData [<rgba 0.1125 0.2 0.05 1>, <rgba 0.1387 0.2 0.05 1>, <rgba 0.114 0.1683 0.0337 1>, <rgba 0.1857 0.2 0.05 1>]}}
     Random Seed=0
    )
    Lifetime Max=1
    Lifetime Min=0.6
    Lifetime Mode=::NewEnumerator1
    Sprite Rotation Mode=::NewEnumerator2
    Sprite Size Max=<12.5 7.5>
    Sprite Size Min=<7.5 2.5>
    Sprite Size Mode=::NewEnumerator2
    UsePositionOffset=false
    Write Material Random=true
   ) {module=InitializeParticle001}
   &SocketLocation(
    Interpolate Socket Sampling=true
    SamplePosOffset=false
    Socket Reader=$System.Socket Reader
    Socket Sampling Mode=::NewEnumerator3
    VelocitySpeedLimit=false
   )
   &InheritVelocity(Speed Limit=true, UseSpeedThreshold=false)
   &DynamicMaterialParameters(
    Index 0 Param 1=&FN_FrontEndFloatModifier(
     Gameplay=&FNTimeOfDayScalar(Night Scale=0.06, Morning Scale=0.06, Day Scale=0.1, Evening Scale=0.08)
     Front End=0.01
    )
    Module.Param0WriteEnabled=true
    Module.Param1WriteEnabled=true
    Module.Param2WriteEnabled=true
    Module.Param3WriteEnabled=true
   ) {module=DynamicMaterialParameters001}
   &InitialMeshOrientation(Use Rotation Vector=false)
  ]
  particleUpdateScript [
   &ParticleState()
   &ScaleSpriteSize(Uniform Curve Sprite Scale=DataInterface() {propertyValues {Curve {
       defaultValue=1
       keys [
        {arriveTangent=7.4671, interpMode=RCIM_Cubic, leaveTangent=7.4671, tangentMode=RCTM_User, time=0, value=0.5}
        {arriveTangent=0.0097, interpMode=RCIM_Cubic, leaveTangent=0.0097, tangentMode=RCTM_User, time=0.1, value=1}
        {interpMode=RCIM_Cubic, tangentMode=RCTM_User, time=0.5, value=1}
        {tangentMode=RCTM_User, time=1, value=0}
       ]
      }}})
   &ScaleSpriteSize(
    Initial Sprite Size:Vector2f=$Particles.SpriteSize
    Uniform Curve Index=/Game/Effects/Niagara/Modules/Math/FN_CameraDistance(
     0Dist:NiagaraFloat=70
     1Dist:NiagaraFloat=90
     Camera Query:NiagaraDataInterfaceCamera=DataInterface()
     Position:NiagaraPosition=$Particles.Position
    )
   ) {module=ScaleSpriteSize001}
   &UpdateMeshOrientation(
    Rotation Rate=&FloatFromCurve(
     FloatCurve=DataInterface() {propertyValues {Curve {
        defaultValue=1
        keys [{interpMode=RCIM_Cubic, tangentMode=RCTM_User, time=0, value=1}, {interpMode=RCIM_Cubic, tangentMode=RCTM_User, time=1, value=0}]
       }}}
     Scale Curve=&RandomRangeFloat(Minimum=0.5, Maximum=1)
    )
    Rotation Vector=&RandomVector(Vector Scale=1)
   )
   &SubUVAnimation(
    Random Start Frame=true
    "Sprite Renderer "=DataInterface() {propertyValues {SpriteRenderer=&self:NE_Leaves_0.NiagaraSpriteRendererProperties_2}}
    SubUV Lookup Index Scale=0.5
    UseEndFrame=true
    UseStartFrame=true
   ) {module=SubUVAnimation002}
   /Niagara/Modules/Update/Forces/CurlNoiseForce(
    Noise Frequency=100
    Noise Strength=&Multiply_Float(
     A=100
     B=&FloatFromCurve(FloatCurve=DataInterface() {propertyValues {Curve {keys [{interpMode=RCIM_Cubic, tangentMode=RCTM_User, time=0, value=0}, {interpMode=RCIM_Cubic, tangentMode=RCTM_User, time=1, value=1}]}}})
    )
    Pan Noise=false
   )
   &GravityForce(Gravity=&RandomRangeVector(Minimum=<0 0 -300>, Maximum=<0 0 -400>))
   &AerodynamicDrag(
    Aerodynamic Drag=0.5
    Aerodynamic Rotational Drag=0.5
    Drag Shape Curve=DataInterface() {propertyValues {Curve {keys [
        {interpMode=RCIM_Cubic, tangentMode=RCTM_User, time=0, value=0.1}
        {arriveTangent=2.0799, interpMode=RCIM_Cubic, leaveTangent=2.0799, tangentMode=RCTM_User, time=1, value=1}
       ]}}}
    Lift Shape Curve=DataInterface() {propertyValues {Curve {keys [
        {interpMode=RCIM_Cubic, tangentMode=RCTM_User, time=0, value=0}
        {interpMode=RCIM_Cubic, tangentMode=RCTM_User, time=0.5, value=1}
        {tangentMode=RCTM_User, time=1, value=0}
       ]}}}
    Use Linear Drag=true
    Use Rotational Drag=true
   )
   &SolveForcesAndVelocity(Clamp Velocity=false, Limit Acceleration=false)
   &AlignSpriteToMeshOrientation(SetAlignment=true, SetFacing=true)
   &ScaleColor(
    Scale RGB=&VectorFromFloat(Value=&FN_FrontEndFloatModifier(
      Gameplay=&FNTimeOfDayScalar(Night Scale=0.3, Morning Scale=0.7, Day Scale=0.5, Evening Scale=0.7)
      Front End=0.31
     ))
    ScaleA=true
    ScaleRGB=true
   )
  ]
  rendererClasses [&NiagaraSpriteRendererProperties]
  renderers [{rendererClass=&NiagaraSpriteRendererProperties}]
 }
 NE_Trail_Head {
  emitterSpawnScript [&FN_RibbonIDOffsetUtility()]
  emitterUpdateScript [
   &EmitterState()
   &FN_SpawnAnimTrail()
  ]
  particleSpawnScript [
   &FN_InitializeAnimTrailBoneIndecies(First Bone or Socket Index=3, Second Bone or Socket Index=4)
   &FN_InitializeAnimTrail(AnimTrailWidth=2, Lifetime=0.2)
  ]
  particleUpdateScript [
   &SolveForcesAndVelocity(Clamp Velocity=false, Limit Acceleration=false)
   &FN_UpdateAnimTrail() {module=FN_UpdateAnimTrail001}
   &ScaleColor(
    Color Value To Scale=$Particles.Color
    Scale Alpha=&FloatFromCurve(
     FloatCurve=DataInterface() {propertyValues {Curve {
        defaultValue=1
        keys [
         {interpMode=RCIM_Cubic, tangentMode=RCTM_User, time=0, value=1}
         {arriveTangent=-0.0587, interpMode=RCIM_Cubic, leaveTangent=-0.0587, tangentMode=RCTM_User, time=0.4, value=0}
        ]
       }}}
     Scale Curve=&FNTimeOfDayScalar(Night Scale=0.75, Morning Scale=1.5, Day Scale=1.5, Evening Scale=1.5)
    )
    ScaleA=true
    ScaleRGB=true
   )
   &ScaleRibbonWidth(Ribbon Width Scale=&FloatFromCurve(FloatCurve=DataInterface() {propertyValues {Curve {keys [<at 0 1>,<at 1 0>]}}}))
  ]
  rendererClasses [&NiagaraRibbonRendererProperties]
  renderers [{rendererClass=&NiagaraRibbonRendererProperties}]
 }
 NE_Trail_Vines {
  emitterSpawnScript [&FN_RibbonIDOffsetUtility()]
  emitterUpdateScript [
   &EmitterState()
   &FN_SpawnAnimTrail(Spawn Interval=0)
   /SetVariables(
    Emitter.HighColor:LinearColor=&FNTimeOfDayColor(Night Color=<rgba 0.1875 0.3 0.075 1>, Morning Color=<rgba 0.3 0.2 0 1>, Day Color=<rgba 0.2087 0.3 0 1>, Evening Color=<rgba 0.25 0.3 0 1>)
    Emitter.LowColor:LinearColor=&FNTimeOfDayColor(Night Color=<rgba 0.225 0.3 0.075 1>, Morning Color=<rgba 0.3 0.2 0 1>, Day Color=<rgba 0.1617 0.3 0 1>, Evening Color=<rgba 0.2 0.3 0 1>)
   ) {isSetParameters=true}
  ]
  particleSpawnScript [
   &FN_InitializeAnimTrailBoneIndecies(First Bone or Socket Index=3, Second Bone or Socket Index=4)
   &FN_InitializeAnimTrail(AnimTrailWidth=3, Lifetime=0.175)
  ]
  particleUpdateScript [
   &SolveForcesAndVelocity(Clamp Velocity=false, Limit Acceleration=false)
   &FN_UpdateAnimTrail() {module=FN_UpdateAnimTrail001}
   &ScaleColor(
    Scale Alpha=&FloatFromCurve(
     FloatCurve=DataInterface() {propertyValues {Curve {
        defaultValue=1
        keys [
         {interpMode=RCIM_Cubic, tangentMode=RCTM_User, time=0, value=1}
         {arriveTangent=-0.0587, interpMode=RCIM_Cubic, leaveTangent=-0.0587, tangentMode=RCTM_User, time=0.8, value=0}
        ]
       }}}
     Scale Curve=&Multiply_Float(A=$System.SystemFade, B=5)
    )
    ScaleA=true
    ScaleRGB=true
   )
   &ScaleRibbonWidth(Ribbon Width Scale=&FloatFromCurve(FloatCurve=DataInterface() {propertyValues {Curve {
        defaultValue=1
        keys [
         {arriveTangent=1.7969, interpMode=RCIM_Cubic, leaveTangent=1.7969, tangentMode=RCTM_User, time=0, value=0.4}
         {arriveTangent=-0.0146, interpMode=RCIM_Cubic, leaveTangent=-0.0146, tangentMode=RCTM_User, time=0.8, value=1}
         {arriveTangent=-9.4017, interpMode=RCIM_Cubic, leaveTangent=-9.4017, tangentMode=RCTM_User, time=1, value=0}
        ]
       }}}))
  ]
  rendererClasses [&NiagaraRibbonRendererProperties]
  renderers [{rendererClass=&NiagaraRibbonRendererProperties}]
 }
}
summary {
 systemName=NS_Pickaxe_CrimpPerfume_Trail
 emitters [
  {emitterName=NE_Leaves, rendererClasses [&NiagaraSpriteRendererProperties]}
  {emitterName=NE_Trail_Head, rendererClasses [&NiagaraRibbonRendererProperties]}
  {emitterName=NE_Trail_Vines, rendererClasses [&NiagaraRibbonRendererProperties]}
 ]
 userVariables [
  User.bIsFrontEnd {type=NiagaraBool}
  User.bIsFrontEndPreview {type=NiagaraBool}
  User.RibbonIDOffset {type=NiagaraInt32}
 ]
}
